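# Physics asset for the female medium Kepler jumpsuit, giving bodies and joint limits to its dangling details: back tags, front zippers, shoelaces and backside strings.
# Small base bodies anchor box tags and zippers, plus capsule lace bows and two-segment lace strings.
# Two four-link backside string chains share one parent, with swing limits widening toward the tips.
# Foot capsules are the only non-dynamic-named bodies and have no constraints.
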PhysicsAsset /BRCosmetics/Characters/Player/Female/Medium/Bodies/F_MED_Kepler_Jumpsuit/Meshes/F_MED_Kepler_Jumpsuit_Physics
bodies {
 dyn_back_tag_base_l [{shapeType=Box, extentX=0.5, extentY=0.5, extentZ=0.5}]
 dyn_back_tag_base_r [{shapeType=Box, extentX=0.5, extentY=0.5, extentZ=0.5}]
 dyn_back_tag_parent_l [{shapeType=Box, extentX=3.1757, extentY=3.5, extentZ=1.0913, center=<1.3946 0 -0.0001>}]
 dyn_back_tag_parent_r [{shapeType=Box, extentX=3.1757, extentY=3.25, extentZ=1.3413, center=<1.2618 0 0>}]
 dyn_shoelace_base_l [{shapeType=Sphere, radius=0.5}]
 dyn_shoelace_base_r [{shapeType=Sphere, radius=0.5}]
 dyn_shoelace_bow_inner_l [{shapeType=Capsule, length=1.7185, radius=1.4116, center=<1.2256 0 0>, rotation=<rot 90 0 0>}]
 dyn_shoelace_bow_inner_r [{shapeType=Capsule, length=1.7339, radius=1.333, center=<1.9615 0 0>, rotation=<rot 90 0 0>}]
 dyn_shoelace_bow_outter_l [{shapeType=Capsule, length=1.9935, radius=1.2597, center=<1.6918 0 0>, rotation=<rot 90 0 0>}]
 dyn_shoelace_bow_outter_r [{shapeType=Capsule, length=1.3938, radius=1.2414, center=<1.5373 0 0>, rotation=<rot 90 0 0>}]
 dyn_shoelace_string_01_inner_l [{shapeType=Capsule, length=1.4419, radius=1, center=<1.4985 0 0>, rotation=<rot 90 0 0>}]
 dyn_shoelace_string_01_inner_r [{shapeType=Capsule, length=2.4383, radius=0.75, center=<1.809 0 0>, rotation=<rot 90 0 0>}]
 dyn_shoelace_string_01_outter_l [{shapeType=Capsule, length=1.3634, radius=1, center=<1.412 0 -0.0001>, rotation=<rot 90 0 0>}]
 dyn_shoelace_string_01_outter_r [{shapeType=Capsule, length=1.866, radius=0.75, center=<1.5894 0 0>, rotation=<rot 90 0 0>}]
 dyn_shoelace_string_02_inner_l [{shapeType=Capsule, length=2.7993, radius=1, center=<2.1652 0 0.0001>, rotation=<rot 90 0 0>}]
 dyn_shoelace_string_02_inner_r [{shapeType=Capsule, length=3.505, radius=0.755, center=<2.2517 0 0>, rotation=<rot 90 0 0>}]
 dyn_shoelace_string_02_outter_l [{shapeType=Capsule, length=3.005, radius=1.005, center=<1.9954 0 0>, rotation=<rot 90 0 0>}]
 dyn_shoelace_string_02_outter_r [{shapeType=Capsule, length=2.8311, radius=0.75, center=<1.9674 -0.0001 0>, rotation=<rot 90 0 0>}]
 dyn_string_backside_01_l [{shapeType=Capsule, length=1.005, radius=1.005, center=<1.3499 0 -0.0001>, rotation=<rot 90 0 0>}]
 dyn_string_backside_01_r [{shapeType=Capsule, length=1.63, radius=0.755, center=<1.2391 0 0.0001>, rotation=<rot 90 0 0>}]
 dyn_string_backside_02_l [{shapeType=Capsule, length=2.505, radius=1.005, center=<2.115 0 0.0001>, rotation=<rot 90 0 0>}]
 dyn_string_backside_02_r [{shapeType=Capsule, length=3.0993, radius=0.75, center=<2.1373 0 0>, rotation=<rot 90 0 0>}]
 dyn_string_backside_03_l [{shapeType=Capsule, length=2.13, radius=1.005, center=<1.5187 0 0>, rotation=<rot 90 0 0>}]
 dyn_string_backside_03_r [{shapeType=Capsule, length=2.505, radius=0.755, center=<1.6723 0 0>, rotation=<rot 90 0 0>}]
 dyn_string_backside_04_l [{shapeType=Capsule, length=1.8112, radius=1, center=<1.3912 0 -0.0001>, rotation=<rot 90 0 0>}]
 dyn_string_backside_04_r [{shapeType=Capsule, length=1.88, radius=0.755, center=<1.1394 0 -0.0001>, rotation=<rot 90 0 0>}]
 dyn_string_backside_parent [{shapeType=Sphere, radius=0.5}]
 dyn_zipper_lower_front_base_l [{shapeType=Sphere, radius=0.5}]
 dyn_zipper_lower_front_base_r [{shapeType=Sphere, radius=0.5}]
 dyn_zipper_lower_front_parent_l [{shapeType=Box, extentX=5.0661, extentY=2.1274, extentZ=1, center=<2.0639 0 0>}]
 dyn_zipper_lower_front_parent_r [{shapeType=Box, extentX=5.0556, extentY=2.7154, extentZ=1.2726, center=<2.2805 0 0>}]
 dyn_zipper_mid_front_base_l [{shapeType=Sphere, radius=0.5}]
 dyn_zipper_mid_front_base_r [{shapeType=Sphere, radius=0.5}]
 dyn_zipper_mid_front_parent_l [{shapeType=Box, extentX=6.6515, extentY=2.7587, extentZ=1.1095, center=<2.4719 0 0>}]
 dyn_zipper_mid_front_parent_r [{shapeType=Box, extentX=5.6628, extentY=2.1672, extentZ=1.1095, center=<2.523 0 0>}]
 dyn_zipper_top_front_base_l [{shapeType=Sphere, radius=0.5}]
 dyn_zipper_top_front_base_r [{shapeType=Sphere, radius=0.5}]
 dyn_zipper_top_front_parent_l [{shapeType=Box, extentX=6.6585, extentY=3.2325, extentZ=2.0444, center=<2.9564 0 0>}]
 dyn_zipper_top_front_parent_r [{shapeType=Box, extentX=5.6517, extentY=2.2891, extentZ=2.0445, center=<2.9131 0 -0.0001>}]
 foot_l [{shapeType=Capsule, length=10.25, radius=4, center=<1.2116 2.1014 0>, rotation=<rot 90 0 21.2006>}]
 foot_r [{shapeType=Capsule, length=10.25, radius=4, center=<-0.7633 -2.1373 -0.0003>, rotation=<rot 90 165.9637 -168.9884>}]
}
constraints [
 {bone1Name=dyn_string_backside_01_r, bone2Name=dyn_string_backside_parent, swing1LimitDegrees=10, swing2LimitDegrees=20, twistLimitDegrees=5}
 {bone1Name=dyn_string_backside_02_r, bone2Name=dyn_string_backside_01_r, swing1LimitDegrees=15, swing2LimitDegrees=20, twistLimitDegrees=5}
 {bone1Name=dyn_string_backside_03_r, bone2Name=dyn_string_backside_02_r, swing1LimitDegrees=20, swing2LimitDegrees=20, twistLimitDegrees=5}
 {bone1Name=dyn_string_backside_04_r, bone2Name=dyn_string_backside_03_r, swing1LimitDegrees=25, swing2LimitDegrees=20, twistLimitDegrees=5}
 {bone1Name=dyn_string_backside_01_l, bone2Name=dyn_string_backside_parent, swing1LimitDegrees=15, swing2LimitDegrees=20, twistLimitDegrees=5}
 {bone1Name=dyn_string_backside_02_l, bone2Name=dyn_string_backside_01_l, swing1LimitDegrees=20, swing2LimitDegrees=20, twistLimitDegrees=5}
 {bone1Name=dyn_string_backside_03_l, bone2Name=dyn_string_backside_02_l, swing1LimitDegrees=25, swing2LimitDegrees=20, twistLimitDegrees=5}
 {bone1Name=dyn_string_backside_04_l, bone2Name=dyn_string_backside_03_l, swing1LimitDegrees=30, swing2LimitDegrees=20, twistLimitDegrees=5}
 {bone1Name=dyn_back_tag_parent_l, bone2Name=dyn_back_tag_base_l, swing1LimitDegrees=10, swing2LimitDegrees=15, twistLimitDegrees=3}
 {bone1Name=dyn_back_tag_parent_r, bone2Name=dyn_back_tag_base_r, swing1LimitDegrees=10, swing2LimitDegrees=15, twistLimitDegrees=3}
 {bone1Name=dyn_shoelace_bow_inner_r, bone2Name=dyn_shoelace_base_r, swing1LimitDegrees=20, swing2LimitDegrees=20, twistLimitDegrees=3}
 {bone1Name=dyn_shoelace_bow_outter_l, bone2Name=dyn_shoelace_base_l, swing1LimitDegrees=20, swing2LimitDegrees=20, twistLimitDegrees=3}
 {bone1Name=dyn_shoelace_string_01_inner_r, bone2Name=dyn_shoelace_base_r, swing1LimitDegrees=10, swing2LimitDegrees=10, twistLimitDegrees=3}
 {
  bone1Name=dyn_shoelace_string_02_inner_r
  bone2Name=dyn_shoelace_string_01_inner_r
  swing1LimitDegrees=20
  swing2LimitDegrees=20
  twistLimitDegrees=5
 }
 {bone1Name=dyn_shoelace_bow_outter_r, bone2Name=dyn_shoelace_base_r, swing1LimitDegrees=20, swing2LimitDegrees=20, twistLimitDegrees=3}
 {bone1Name=dyn_shoelace_string_01_outter_r, bone2Name=dyn_shoelace_base_r, swing1LimitDegrees=10, swing2LimitDegrees=10, twistLimitDegrees=3}
 {
  bone1Name=dyn_shoelace_string_02_outter_r
  bone2Name=dyn_shoelace_string_01_outter_r
  swing1LimitDegrees=20
  swing2LimitDegrees=20
  twistLimitDegrees=5
 }
 {bone1Name=dyn_shoelace_bow_inner_l, bone2Name=dyn_shoelace_base_l, swing1LimitDegrees=20, swing2LimitDegrees=20, twistLimitDegrees=3}
 {
  bone1Name=dyn_shoelace_string_02_outter_l
  bone2Name=dyn_shoelace_string_01_outter_l
  swing1LimitDegrees=20
  swing2LimitDegrees=20
  twistLimitDegrees=5
 }
 {bone1Name=dyn_shoelace_string_01_outter_l, bone2Name=dyn_shoelace_base_l, swing1LimitDegrees=10, swing2LimitDegrees=10, twistLimitDegrees=3}
 {
  bone1Name=dyn_shoelace_string_02_inner_l
  bone2Name=dyn_shoelace_string_01_inner_l
  swing1LimitDegrees=20
  swing2LimitDegrees=20
  twistLimitDegrees=5
 }
 {bone1Name=dyn_shoelace_string_01_inner_l, bone2Name=dyn_shoelace_base_l, swing1LimitDegrees=10, swing2LimitDegrees=10, twistLimitDegrees=3}
 {
  bone1Name=dyn_zipper_mid_front_parent_r
  bone2Name=dyn_zipper_mid_front_base_r
  swing1LimitDegrees=10
  swing2LimitDegrees=10
  twistLimitDegrees=3
 }
 {
  bone1Name=dyn_zipper_top_front_parent_r
  bone2Name=dyn_zipper_top_front_base_r
  swing1LimitDegrees=20
  swing2LimitDegrees=20
  twistLimitDegrees=5
 }
 {
  bone1Name=dyn_zipper_lower_front_parent_r
  bone2Name=dyn_zipper_lower_front_base_r
  swing1LimitDegrees=20
  swing2LimitDegrees=20
  twistLimitDegrees=3
 }
 {
  bone1Name=dyn_zipper_top_front_parent_l
  bone2Name=dyn_zipper_top_front_base_l
  swing1LimitDegrees=20
  swing2LimitDegrees=20
  twistLimitDegrees=3
 }
 {
  bone1Name=dyn_zipper_mid_front_parent_l
  bone2Name=dyn_zipper_mid_front_base_l
  swing1LimitDegrees=20
  swing2LimitDegrees=20
  twistLimitDegrees=3
 }
 {
  bone1Name=dyn_zipper_lower_front_parent_l
  bone2Name=dyn_zipper_lower_front_base_l
  swing1LimitDegrees=20
  swing2LimitDegrees=20
  twistLimitDegrees=3
 }
]
