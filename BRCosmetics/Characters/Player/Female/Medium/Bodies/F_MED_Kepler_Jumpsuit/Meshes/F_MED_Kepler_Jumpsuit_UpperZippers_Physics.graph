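# Physics asset for the upper zipper pulls on the female medium Kepler jumpsuit cosmetic, covering back, pocket and jacket zippers.
# Each of the five zippers pairs a tiny sphere body on its base bone with an offset box body on its parent bone.
# A constraint links each pair, allowing moderate swing with very tight twist so the pulls can dangle without spinning.
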PhysicsAsset /BRCosmetics/Characters/Player/Female/Medium/Bodies/F_MED_Kepler_Jumpsuit/Meshes/F_MED_Kepler_Jumpsuit_UpperZippers_Physics
bodies {
 dyn_zipper_back_base_l [{shapeType=Sphere, radius=0.5}]
 dyn_zipper_back_base_r [{shapeType=Sphere, radius=0.5}]
 dyn_zipper_back_parent_l [{shapeType=Box, extentX=5.7661, extentY=2.2313, extentZ=0.8604, center=<2.8327 0 0.0001>}]
 dyn_zipper_back_parent_r [{shapeType=Box, extentX=5.757, extentY=2.327, extentZ=0.8604, center=<2.8327 0 0.0001>}]
 dyn_zipper_jacket_base_mid [{shapeType=Sphere, radius=0.5}]
 dyn_zipper_jacket_parent_mid [{shapeType=Box, extentX=5.1268, extentY=1.2614, extentZ=1.8545, center=<2.2831 0 0>}]
 dyn_zipper_pocket_base_l [{shapeType=Sphere, radius=0.5}]
 dyn_zipper_pocket_base_r [{shapeType=Sphere, radius=0.5}]
 dyn_zipper_pocket_parent_l [{shapeType=Box, extentX=4.7612, extentY=1.6073, extentZ=1.3554, center=<2.4232 0 0>}]
 dyn_zipper_pocket_parent_r [{shapeType=Box, extentX=5.7566, extentY=1.671, extentZ=1.3554, center=<2.8754 0 0>}]
}
constraints [
 {bone1Name=dyn_zipper_back_parent_r, bone2Name=dyn_zipper_back_base_r, swing1LimitDegrees=35, swing2LimitDegrees=35, twistLimitDegrees=3}
 {bone1Name=dyn_zipper_back_parent_l, bone2Name=dyn_zipper_back_base_l, swing1LimitDegrees=40, swing2LimitDegrees=40, twistLimitDegrees=3}
 {bone1Name=dyn_zipper_pocket_parent_l, bone2Name=dyn_zipper_pocket_base_l, swing1LimitDegrees=40, swing2LimitDegrees=35, twistLimitDegrees=3}
 {bone1Name=dyn_zipper_pocket_parent_r, bone2Name=dyn_zipper_pocket_base_r, swing1LimitDegrees=40, swing2LimitDegrees=35, twistLimitDegrees=3}
 {
  bone1Name=dyn_zipper_jacket_parent_mid
  bone2Name=dyn_zipper_jacket_base_mid
  swing1LimitDegrees=30
  swing2LimitDegrees=30
  twistLimitDegrees=3
 }
]
