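# Animation Blueprint for the female medium Kepler jumpsuit cosmetic, layering secondary motion onto the attached parent mesh's copied pose.
# Additive bone tweaks offset two badges and pitch the left collar while targeting or parachuting.
# Weighted constraints tie back tag and string bases to pelvis and thighs, then two RigidBody passes simulate the suit and upper zippers.
# Jump events set and clear a delayed upward external force feeding both simulations.
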AnimBlueprint /BRCosmetics/Characters/Player/Female/Medium/Bodies/F_MED_Kepler_Jumpsuit/Meshes/F_MED_Kepler_Jumpsuit_AnimBP
target_skeleton=&/F_MED_Kepler_Jumpsuit_Skeleton
graphs [
 AnimGraph {nodes [
   n1 = Get ExternalForce() -> (ExternalForce:Vector)
   n2 = "Transform (Modify) Bone - Bone: const_badge_01"(
    ComponentPose=Local To Component(LocalPose=Copy Pose From Mesh() {bUseAttachedParent=true})
    Alpha=1
    Translation="0.019,0.723,1.110"
    Rotation="-0.22,0.005,-2.743"
    Scale="1,1,1"
   ) {
    rotationMode=BMM_Additive
    rotationSpace=BCS_ParentBoneSpace
    translationMode=BMM_Additive
    translationSpace=BCS_ParentBoneSpace
    alphaBoolBlend {blendOption=Linear}
    alphaScaleBiasClamp {bInterpResult=true, interpSpeedDecreasing=4, interpSpeedIncreasing=6}
    boneToModify {boneName=const_badge_01}
    rotation=<rot -0.22 0.005 -2.743>
    translation=<0.019 0.723 1.11>
   }
   n3 = "Transform (Modify) Bone - Bone: const_badge_02"(ComponentPose=@n2, Alpha=1, Translation="-0.003,1.012,1.479", Rotation="-0.499,0.740,-6.658", Scale="1,1,1") {
    rotationMode=BMM_Additive
    rotationSpace=BCS_ParentBoneSpace
    translationMode=BMM_Additive
    translationSpace=BCS_ParentBoneSpace
    alphaBoolBlend {blendOption=Linear}
    alphaScaleBiasClamp {bInterpResult=true, interpSpeedDecreasing=4, interpSpeedIncreasing=6}
    boneToModify {boneName=const_badge_02}
    rotation=<rot -0.499 0.74 -6.658>
    translation=<-0.003 1.012 1.479>
   }
   n4 = "Transform (Modify) Bone - Bone: dyn_collar_l"(ComponentPose=@n3, Alpha=Get Collar() -> (Collar:double), Translation="0,0,0", Rotation="50,0,0", Scale="1,1,1") {
    rotationMode=BMM_Additive
    rotationSpace=BCS_ParentBoneSpace
    translationMode=BMM_Additive
    translationSpace=BCS_ParentBoneSpace
    alphaBoolBlend {blendOption=Linear}
    alphaScaleBiasClamp {bInterpResult=true, interpSpeedDecreasing=2, interpSpeedIncreasing=4}
    boneToModify {boneName=dyn_collar_l}
    rotation {pitch=50}
   }
   n5 = "Constraint - dyn_back_tag_base_l (PPPPP)"!(
    ComponentPose=@n4
    Alpha=1
    ConstraintWeights_0=0
    ConstraintWeights_1:float=0.7327
    ConstraintWeights_2:float=0.2504
    ConstraintWeights_3:float=0.0001
    ConstraintWeights_4:float=0.0168
   ) -> (Pose) {
    alphaBoolBlend {blendOption=Linear}
    boneToModify {boneName=dyn_back_tag_base_l}
    constraintSetup [
     {targetBone {boneName=thigh_r}}
     {targetBone {boneName=pelvis}}
     {targetBone {boneName=thigh_twist_01_l}}
     {targetBone {boneName=thigh_twist_01_r}}
     {targetBone {boneName=thigh_l}}
    ]
    constraintWeights [1,1,1,1,1]
   }
   n6 = "Constraint - dyn_back_tag_base_r (PPP)"!(ComponentPose=@n5, Alpha=1, ConstraintWeights_0=0.0128, ConstraintWeights_1:float=0.7468, ConstraintWeights_2:float=0.2405) -> (Pose) {
    alphaBoolBlend {blendOption=Linear}
    boneToModify {boneName=dyn_back_tag_base_r}
    constraintSetup [{targetBone {boneName=thigh_r}}, {targetBone {boneName=pelvis}}, {targetBone {boneName=thigh_twist_01_r}}]
    constraintWeights [1,1,1]
   }
   n7 = "Constraint - dyn_string_backside_base (PP)"!(ComponentPose=@n6, Alpha=1, ConstraintWeights_0=0.9612, ConstraintWeights_1:float=0.0388) -> (Pose) {
    alphaBoolBlend {blendOption=Linear}
    boneToModify {boneName=dyn_string_backside_base}
    constraintSetup [{targetBone {boneName=pelvis}}, {targetBone {boneName=thigh_twist_01_l}}]
    constraintWeights [1,1]
   }
   n8 = RigidBody(ComponentPose=@n7, ExternalForce=@n1) {
    bClampLinearTranslationLimitToRefPose=true
    bOverrideWorldGravity=true
    overridePhysicsAsset=&/F_MED_Kepler_Jumpsuit_Physics
    simulationSpace=BaseBoneSpace
    alphaBoolBlend {blendOption=Linear}
    alphaScaleBiasClamp {bInterpResult=true, interpSpeedDecreasing=4, interpSpeedIncreasing=6}
    baseBoneRef {boneName=root}
    componentAppliedLinearAccClamp=<6000 6000 6000>
    componentLinearAccScale=<0.2 0.2 0.2>
    componentLinearVelScale=<0.2 0.2 0.2>
    overrideWorldGravity {z=-980}
   }
   n9 = RigidBody(ComponentPose=@n8, ExternalForce=@n1) {
    bClampLinearTranslationLimitToRefPose=true
    bOverrideWorldGravity=true
    overridePhysicsAsset=&/F_MED_Kepler_Jumpsuit_UpperZippers_Physics
    simulationSpace=BaseBoneSpace
    alphaBoolBlend {blendOption=Linear}
    alphaScaleBiasClamp {bInterpResult=true, interpSpeedDecreasing=4, interpSpeedIncreasing=6}
    baseBoneRef {boneName=pelvis}
    componentAppliedLinearAccClamp=<6000 6000 6000>
    componentLinearAccScale=<0.2 0.2 0.2>
    componentLinearVelScale=<0.2 0.2 0.2>
    overrideWorldGravity {z=-980}
   }
   Output Pose(Result=Component To Local(ComponentPose=@n9)) {layerGroup=None}
  ]}
 EventGraph {nodes [
   Event BlueprintUpdateAnimation()
   n1 = Get bIsTargeting()
   n2 = Get bIsParachuteOpened()
   > Set Collar(Collar:double=SelectFloat(A=1, B=0, bPickA=OR Boolean(A=@n1, B=@n2))) -> (Output_Get:double)
   Event OnBeginJumping()
   > RetriggerableDelay(Duration=0.2)
   > Set ExternalForce(ExternalForce:Vector="0,0,1250") -> (Output_Get:Vector)
   Event OnEndJumping()
   > RetriggerableDelay(Duration=0.1)
   > Set ExternalForce(ExternalForce="0, 0, 0")
  ]}
]
variables [
 Body_Mesh {type=SkeletalMeshComponent}
 Collar {type=number}
 ExternalForce {type=Vector}
]
