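# Physics asset for the female medium Kepler Jumpsuit cosmetic that sets up secondary motion for the character's hair.
# Back, side, side-back, front and strand hair chains use capsules hanging from small sphere base bodies, linked in sequence by swing and twist angle limits.
# Capsules on head, neck_02 and spine_05 have no constraints, apparently serving as collision volumes for the hair.
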PhysicsAsset /BRCosmetics/Characters/Player/Female/Medium/Bodies/F_MED_Kepler_Jumpsuit/Meshes/Parts/F_MED_Kepler_Jumpsuit_Physics
bodies {
 dyn_hair_back_01_mid [{shapeType=Capsule, length=2.9147, radius=2.2896, center=<2.4396 -0.0001 -0.0001>, rotation=<rot 90 0 0>}]
 dyn_hair_back_02_mid [{shapeType=Capsule, length=2.3012, radius=2.8195, center=<3.6998 0 0>, rotation=<rot 90 0 0>}]
 dyn_hair_back_base_mid [{shapeType=Sphere, radius=0.5}]
 dyn_hair_back_parent_mid [{shapeType=Capsule, length=2.1059, radius=2.2764, center=<2.5175 0 0.0002>, rotation=<rot 90 0 0>}]
 dyn_hair_front_01_l [{shapeType=Capsule, length=2.3139, radius=2.3199, center=<2.1424 0 0>, rotation=<rot 90 0 0>}]
 dyn_hair_front_01_r [{shapeType=Capsule, length=1.702, radius=2.1164, center=<1.409 0 0>, rotation=<rot 90 0 0>}]
 dyn_hair_front_base_l [{shapeType=Sphere, radius=0.5}]
 dyn_hair_front_base_r [{shapeType=Sphere, radius=0.5}]
 dyn_hair_front_parent_l [{shapeType=Capsule, length=2.903, radius=2.0978, center=<1.9346 0 -0.0002>, rotation=<rot 90 0 0>}]
 dyn_hair_front_parent_r [{shapeType=Capsule, length=1.4948, radius=2.3927, center=<1.8992 0 0>, rotation=<rot 90 0 0>}]
 dyn_hair_front_strand_base_l [{shapeType=Sphere, radius=0.5}]
 dyn_hair_front_strand_base_r [{shapeType=Sphere, radius=0.5}]
 dyn_hair_front_strand_parent_l [{shapeType=Capsule, length=1.8084, radius=1.21, center=<1.2679 0 0>, rotation=<rot 90 0 0>}]
 dyn_hair_front_strand_parent_r [{shapeType=Capsule, length=2.2298, radius=1.1686, center=<0.7657 0 0>, rotation=<rot 90 0 0>}]
 dyn_hair_side_01_l [{shapeType=Capsule, length=2.1413, radius=2.7359, center=<3.2964 0 -0.0002>, rotation=<rot 90 0 0>}]
 dyn_hair_side_01_r [{shapeType=Capsule, length=1.6692, radius=1.8422, center=<2.0037 0 0>, rotation=<rot 90 0 0>}]
 dyn_hair_side_02_r [{shapeType=Capsule, length=2.4597, radius=1.7608, center=<2.2246 0 0>, rotation=<rot 90 0 0>}]
 dyn_hair_side_back_01_l [{shapeType=Capsule, length=1.4003, radius=2.0456, center=<1.9457 0 0>, rotation=<rot 90 0 0>}]
 dyn_hair_side_back_01_r [{shapeType=Capsule, length=1.8971, radius=2.5095, center=<2.6496 0 -0.0001>, rotation=<rot 90 0 0>}]
 dyn_hair_side_back_02_l [{shapeType=Capsule, length=2.2564, radius=2.5763, center=<2.9389 0 0>, rotation=<rot 90 0 0>}]
 dyn_hair_side_back_02_r [{shapeType=Capsule, length=2.6696, radius=2.5273, center=<3.1857 0 0>, rotation=<rot 90 0 0>}]
 dyn_hair_side_back_base_l [{shapeType=Sphere, radius=0.5}]
 dyn_hair_side_back_base_r [{shapeType=Sphere, radius=0.5}]
 dyn_hair_side_back_parent_l [{shapeType=Capsule, length=1.9381, radius=2.2079, center=<2.4821 0 0.0001>, rotation=<rot 90 0 0>}]
 dyn_hair_side_back_parent_r [{shapeType=Capsule, length=2.4004, radius=2.1269, center=<2.4217 0 0>, rotation=<rot 87.133 -5.1781 -7.7235>}]
 dyn_hair_side_base_l [{shapeType=Sphere, radius=0.5}]
 dyn_hair_side_base_r [{shapeType=Sphere, radius=0.5}]
 dyn_hair_side_parent_l [{shapeType=Capsule, length=2.725, radius=2.1019, center=<2.5406 0 0>, rotation=<rot 90 0 0>}]
 dyn_hair_side_parent_r [{shapeType=Capsule, length=1.685, radius=1.7193, center=<1.5825 0 0>, rotation=<rot 90 0 0>}]
 dyn_hair_top_back_strand_base_mid [{shapeType=Sphere, radius=0.5}]
 dyn_hair_top_back_strand_parent_mid [{shapeType=Capsule, length=1, radius=2, center=<-1.2385 0.0001 0>}]
 dyn_hair_top_strand_base_mid [{shapeType=Sphere, radius=0.5}]
 dyn_hair_top_strand_parent_mid [{shapeType=Capsule, length=1.8416, radius=1.4338, center=<-0.7113 0 0>}]
 dyn_head_base [{shapeType=Capsule, length=0.25, radius=0.5, rotation=<rot 0 0 90>}]
 head [{shapeType=Capsule, length=8.5, radius=5.75, center=<0 2.6058 0>, rotation=<rot 90 1.023 -10.73>}]
 neck_02 [{shapeType=Capsule, length=3.5, radius=6.25, center=<0 -0.3858 0>, rotation=<rot 90 0 0>}]
 spine_05 [
  {shapeType=Capsule, length=19.125, radius=10.875, center=<-5.9915 1.3171 -0.4073>, rotation=<rot 90 0 3.6211>}
  {shapeType=Capsule, length=20.75, radius=5.25, center=<4.437 -3.4235 -0.3911>, rotation=<rot -0.0856 0 0>}
 ]
}
constraints [
 {bone1Name=dyn_hair_side_parent_r, bone2Name=dyn_hair_side_base_r, swing1LimitDegrees=20, swing2LimitDegrees=60, twistLimitDegrees=5}
 {bone1Name=dyn_hair_side_01_r, bone2Name=dyn_hair_side_parent_r, swing1LimitDegrees=10, swing2LimitDegrees=10, twistLimitDegrees=5}
 {
  bone1Name=dyn_hair_side_back_parent_r
  bone2Name=dyn_hair_side_back_base_r
  swing1LimitDegrees=20
  swing2LimitDegrees=60
  twistLimitDegrees=5
 }
 {bone1Name=dyn_hair_side_back_01_r, bone2Name=dyn_hair_side_back_parent_r, swing1LimitDegrees=20, swing2LimitDegrees=20, twistLimitDegrees=5}
 {bone1Name=dyn_hair_back_parent_mid, bone2Name=dyn_hair_back_base_mid, swing1LimitDegrees=20, swing2LimitDegrees=45, twistLimitDegrees=5}
 {bone1Name=dyn_hair_back_01_mid, bone2Name=dyn_hair_back_parent_mid, swing1LimitDegrees=15, swing2LimitDegrees=45, twistLimitDegrees=5}
 {bone1Name=dyn_hair_side_parent_l, bone2Name=dyn_hair_side_base_l, swing1LimitDegrees=10, swing2LimitDegrees=20, twistLimitDegrees=5}
 {bone1Name=dyn_hair_side_01_l, bone2Name=dyn_hair_side_parent_l, swing1LimitDegrees=10, swing2LimitDegrees=20, twistLimitDegrees=5}
 {bone1Name=dyn_hair_front_parent_l, bone2Name=dyn_hair_front_base_l, swing1LimitDegrees=20, swing2LimitDegrees=20, twistLimitDegrees=5}
 {bone1Name=dyn_hair_front_01_l, bone2Name=dyn_hair_front_parent_l, swing1LimitDegrees=20, swing2LimitDegrees=20, twistLimitDegrees=5}
 {
  bone1Name=dyn_hair_front_strand_parent_r
  bone2Name=dyn_hair_front_strand_base_r
  swing1LimitDegrees=8
  swing2LimitDegrees=15
  twistLimitDegrees=3
 }
 {
  bone1Name=dyn_hair_front_strand_parent_l
  bone2Name=dyn_hair_front_strand_base_l
  swing1LimitDegrees=10
  swing2LimitDegrees=10
  twistLimitDegrees=3
 }
 {
  bone1Name=dyn_hair_top_strand_parent_mid
  bone2Name=dyn_hair_top_strand_base_mid
  swing1LimitDegrees=8
  swing2LimitDegrees=8
  twistLimitDegrees=3
 }
 {
  bone1Name=dyn_hair_side_back_parent_l
  bone2Name=dyn_hair_side_back_base_l
  swing1LimitDegrees=20
  swing2LimitDegrees=45
  twistLimitDegrees=5
 }
 {bone1Name=dyn_hair_side_back_01_l, bone2Name=dyn_hair_side_back_parent_l, swing1LimitDegrees=20, swing2LimitDegrees=20, twistLimitDegrees=5}
 {bone1Name=dyn_hair_front_01_r, bone2Name=dyn_hair_front_parent_r, swing1LimitDegrees=20, swing2LimitDegrees=20, twistLimitDegrees=5}
 {bone1Name=dyn_hair_front_parent_r, bone2Name=dyn_hair_front_base_r, swing1LimitDegrees=15, swing2LimitDegrees=15, twistLimitDegrees=5}
 {
  bone1Name=dyn_hair_top_back_strand_parent_mid
  bone2Name=dyn_hair_top_back_strand_base_mid
  swing1LimitDegrees=10
  swing2LimitDegrees=10
  twistLimitDegrees=3
 }
 {bone1Name=dyn_hair_back_02_mid, bone2Name=dyn_hair_back_01_mid, swing1LimitDegrees=10, swing2LimitDegrees=20, twistLimitDegrees=5}
 {bone1Name=dyn_hair_side_back_02_r, bone2Name=dyn_hair_side_back_01_r, swing1LimitDegrees=20, swing2LimitDegrees=20, twistLimitDegrees=5}
 {bone1Name=dyn_hair_side_back_02_l, bone2Name=dyn_hair_side_back_01_l, swing1LimitDegrees=20, swing2LimitDegrees=20, twistLimitDegrees=5}
 {bone1Name=dyn_hair_side_02_r, bone2Name=dyn_hair_side_01_r, swing1LimitDegrees=15, swing2LimitDegrees=15, twistLimitDegrees=8}
]
