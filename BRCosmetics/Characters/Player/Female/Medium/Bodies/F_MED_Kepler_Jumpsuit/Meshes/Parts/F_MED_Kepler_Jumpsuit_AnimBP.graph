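# Post-process style AnimBP for the female Kepler Jumpsuit part, copying the attached parent's pose and adjusting dynamic hair while aiming.
# A RigidBody node simulates in spine_05 space, then additive bone transforms offset hair bones with interpolated alphas.
# Targeting drives the first two offsets.
# The last two require targeting with rifles or shotguns plus a head-twist alpha threshold, measured against spine_05 on the body mesh.
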AnimBlueprint /BRCosmetics/Characters/Player/Female/Medium/Bodies/F_MED_Kepler_Jumpsuit/Meshes/Parts/F_MED_Kepler_Jumpsuit_AnimBP
target_skeleton=&/F_MED_Kepler_Jumpsuit_Skeleton
graphs [
 AnimGraph {nodes [
   n1 = RigidBody(ComponentPose=Local To Component(LocalPose=Copy Pose From Mesh() {bUseAttachedParent=true})) {
    bClampLinearTranslationLimitToRefPose=true
    bOverrideWorldGravity=true
    overridePhysicsAsset=&/F_MED_Kepler_Jumpsuit_Physics
    simulationSpace=BaseBoneSpace
    alphaBoolBlend {blendOption=Linear}
    baseBoneRef {boneName=spine_05}
    componentAppliedLinearAccClamp=<6000 6000 6000>
    componentLinearAccScale=<0.2 0.2 0.2>
    componentLinearVelScale=<0.2 0.2 0.2>
    overrideWorldGravity {z=-980}
   }
   n2 = Get ADS_Alpha() -> (ADS_Alpha:double)
   n3 = "Transform (Modify) Bone - Bone: dyn_hair_side_parent_r"(ComponentPose=@n1, Alpha=@n2, Translation="0,-1,-1", Rotation="-40,6.316,0", Scale="1,1,1") {
    rotationMode=BMM_Additive
    rotationSpace=BCS_ParentBoneSpace
    translationMode=BMM_Additive
    translationSpace=BCS_ParentBoneSpace
    alphaBoolBlend {blendOption=Linear}
    alphaScaleBiasClamp {bInterpResult=true, interpSpeedDecreasing=2, interpSpeedIncreasing=4}
    boneToModify {boneName=dyn_hair_side_parent_r}
    rotation {pitch=-40, yaw=6.316}
    translation {y=-1, z=-1}
   }
   n4 = "Transform (Modify) Bone - Bone: dyn_hair_front_base_l"(ComponentPose=@n3, Alpha=@n2, Translation="0,-1,-1", Rotation="-30,0,0", Scale="1,1,1") {
    rotationMode=BMM_Additive
    rotationSpace=BCS_ParentBoneSpace
    translationMode=BMM_Additive
    translationSpace=BCS_ParentBoneSpace
    alphaBoolBlend {blendOption=Linear}
    alphaScaleBiasClamp {bInterpResult=true, interpSpeedDecreasing=2, interpSpeedIncreasing=4}
    boneToModify {boneName=dyn_hair_front_base_l}
    rotation {pitch=-30}
    translation {y=-1, z=-1}
   }
   n5 = Get HeadRotate_Alpha() -> (HeadRotate_Alpha:double)
   n6 = SelectFloat(
    A="To Float (Boolean)"(InBool="float >= float"(A=@n5, B=7) {is=title K2Node_CallFunction})
    B=0
    bPickA=Get Alt_weapon() -> (Alt_weapon:bool)
   )
   n7 = "Transform (Modify) Bone - Bone: dyn_hair_side_back_base_r"(ComponentPose=@n4, Alpha=@n6, Translation="0.000,-4,0", Rotation="0,0,0", Scale="1,1,1") {
    rotationMode=BMM_Additive
    rotationSpace=BCS_ParentBoneSpace
    translationMode=BMM_Additive
    translationSpace=BCS_ParentBoneSpace
    alphaBoolBlend {blendOption=Linear}
    alphaScaleBiasClamp {bInterpResult=true, interpSpeedDecreasing=4, interpSpeedIncreasing=6}
    boneToModify {boneName=dyn_hair_side_back_base_r}
    translation {y=-4}
   }
   n8 = "Transform (Modify) Bone - Bone: dyn_hair_side_parent_r"(ComponentPose=@n7, Alpha=@n6, Translation="0,-3.5,-1", Rotation="-25,40,25", Scale="1,1,1") {
    rotationMode=BMM_Additive
    rotationSpace=BCS_ParentBoneSpace
    translationMode=BMM_Additive
    translationSpace=BCS_ParentBoneSpace
    alphaBoolBlend {blendOption=Linear}
    alphaScaleBiasClamp {bInterpResult=true, interpSpeedDecreasing=2, interpSpeedIncreasing=4}
    boneToModify {boneName=dyn_hair_side_parent_r}
    rotation=<rot -25 40 25>
    translation {y=-3.5, z=-1}
   }
   Output Pose(Result=Component To Local(ComponentPose=@n8)) {layerGroup=None}
  ]}
 EventGraph {nodes [
   Event BlueprintInitializeAnimation()
   > n1 = Cast To PlayerPawn_Athena(Object=TryGetPawnOwner())
   > n2 = Set Pawn Owner(Pawn Owner:Player Pawn Athena Object Reference=@n1) -> (Output_Get:Player Pawn Athena Object Reference)
   > Set Body_Mesh(Body_Mesh=GetSkeletalMeshForPartType(self=@n2))
   Event BlueprintUpdateAnimation()
   > n3 = Sequence!(execute) -> (then_0, then_1, then_2:Exec)
   n4 = Cast To FortPlayerAnimInstance(execute=@n3.then_2, Object=GetAnimInstance(self=GetOwnerSkeletalMeshComponent()))
   > Set Alt_weapon(Alt_weapon:bool=false) -> (Output_Get:bool)
   > n5 = Switch on EFortWeaponCoreAnimation!(execute, Selection:EFortWeaponCoreAnimation Enum=Get WeaponCoreAnim(self=@n4)) -> (
    Melee:Exec
    Pistol:Exec
    Shotgun:Exec
    PaperBlueprint:Exec
    Rifle:Exec
    MeleeOneHand:Exec
    MachinePistol:Exec
    RocketLauncher:Exec
    GrenadeLauncher:Exec
    GoingCommando:Exec
    AssaultRifle:Exec
    TacticalShotgun:Exec
    SniperRifle:Exec
    TrapPlacement:Exec
    ShoulderLauncher:Exec
    AbilityDecoTool:Exec
    Crossbow:Exec
    C4:Exec
    RemoteControl:Exec
    DualWield:Exec
    AR_BullPup:Exec
    AR_ForwardGrip:Exec
    MedPackPaddles:Exec
    SMG_P90:Exec
    AR_DrumGun:Exec
    Consumable_Small:Exec
    Consumable_Large:Exec
    Balloon:Exec
    MountedTurret:Exec
    CreativeTool:Exec
    ExplosiveBow:Exec
    AshtonIndigo:Exec
    AshtonChicago:Exec
    MeleeDualWield:Exec
    Unarmed:Exec
   )
   Branch(execute=[@n5.AssaultRifle, @n5.Rifle, @n5.Shotgun, @n5.TacticalShotgun], Condition=Get bIsTargeting())
   > Set Alt_weapon(Alt_weapon=true)
   Set HeadRotate_Alpha(
    execute=@n3.then_0
    HeadRotate_Alpha:double=GetAlphaFromReferenceTransformDelta(
     SkeletalMeshComponent=Get Body_Mesh()
     BoneName=head
     BaseName=spine_05
     DeltaComponentType=QuaternionTwist
     InputRange=(X=-33,Y=0)
     TwistAxis="0,0,1"
    )
   ) -> (Output_Get:double)
   Set ADS_Alpha(execute=@n3.then_1, ADS_Alpha:double=SelectFloat(A=1, B=0, bPickA=Get bIsTargeting())) -> (Output_Get:double)
  ]}
]
variables [
 ADS_Alpha {type=number}
 Alt_weapon {type=boolean}
 Body_Mesh {type=SkeletalMeshComponent}
 HeadRotate_Alpha {type=number}
 Pawn Owner {type=PlayerPawn_Athena_C}
 Rotate_ADS {type=number}
]
